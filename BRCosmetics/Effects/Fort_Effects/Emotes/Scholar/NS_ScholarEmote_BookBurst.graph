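# Niagara burst effect for the Scholar emote's book, spawning smoke and sparkle sprites on a user-supplied skeletal mesh, tinted by two user colors.
# Smoke puffs rise, slow, spin and expand while curves animate dynamic material parameters.
# Sparkles launch from a squashed sphere, swirl with fading curl noise, pulse in size and glow brightly.
# Particle ID parity varies sparkle size and lifetime and, plus UseGreenMat, sets a VisibilityTag alongside four sprite renderers.
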NiagaraSystem /BRCosmetics/Effects/Fort_Effects/Emotes/Scholar/NS_ScholarEmote_BookBurst
dependencies {usedDataInterfaces [&NiagaraDataInterfaceCurve], usedDynamicInputs [], usedModules [&SystemState]}
emitters {
 Smoke {
  emitterUpdateScript [
   &EmitterState()
   &SpawnBurst_Instantaneous(Spawn Count=7, Spawn Time=0.01, Use Spawn Probability=false) {module=SpawnBurst_Instantaneous001}
  ]
  particleSpawnScript [
   /Niagara/Modules/Spawn/Initialization/InitializeParticle(
    Color=&RandomRangeLinearColor(Minimum=$User.Color_01, Maximum=$User.Color_02)
    Lifetime=&UniformRangedFloat(Minimum=0.6, Maximum=0.8)
    Module.Write Color=true
    Module.Write Lifetime=true
    Module.Write Mass=true
    Module.Write Material Random=true
    Module.Write Position=false
    Module.Write Scale=false
    Module.Write SpriteRotation=true
    Module.Write SpriteSize=true
    Position=&SimulationPosition()
    Sprite Rotation=&UniformRangedFloat(Minimum=0, Maximum=360)
    Sprite Size=&Vector2DFromFloat(Value=&UniformRangedFloat(Minimum=12, Maximum=15))
   )
   &FN_SetSpawnGroup()
   /SetVariables(Particles.InitSpriteRotRate=&UniformRangedFloat(Minimum=-1, Maximum=1)) {isSetParameters=true}
   &SkeletalMeshLocation(
    Bone / Socket Index=$Particles.SpawnGroup
    Bone Sampling Mode=::NewEnumerator4
    Inherit Velocity=true
    "Inherit Velocity (Scale)"=0.5
    Interpolate Mesh Sampling=false
    MaskSpawnGroup=false
    Sampled Position Offset=&RandomRangeVector(Minimum=<-15 -15 -15>, Maximum=<15 15 15>)
    Skeletal Mesh=$User.SkeletalMesh
    VelocitySpeedLimit=false
   )
   &CameraOffset(Camera Offset Amount=&RandomRangeFloat(Minimum=0, Maximum=15))
   &AddVelocity(Velocity=&RandomRangeVector(Minimum=<-10 -10 40>, Maximum=<10 10 60>)) {module=AddVelocity001}
  ]
  particleUpdateScript [
   &ParticleState()
   &SpriteRotationRate(Rotation Rate=&Multiply_Float(
     A=$Particles.InitSpriteRotRate
     B=&FloatFromCurve(
      FloatCurve=DataInterface() {propertyValues {Curve {keys [
          {arriveTangent=-1.9214, interpMode=RCIM_Cubic, leaveTangent=-1.9214, tangentMode=RCTM_User, time=-0.0046, value=1.0113}
          {arriveTangent=-0.2496, interpMode=RCIM_Cubic, leaveTangent=-0.2496, tangentMode=RCTM_User, time=0.9896, value=0.16}
         ]}}}
      CurveIndex=$Emitter.Age
     )
    ))
   &ScaleColor(
    Scale Alpha=&FloatFromCurve(
     FloatCurve=DataInterface() {propertyValues {Curve {keys [<at 0 0 RCIM_Cubic>,<at 0.2049 0.9765 RCIM_Cubic>,<at 1 0 RCIM_Cubic>]}}}
     Scale Curve=0.65
    )
    ScaleA=true
    ScaleRGB=true
   ) {module=ScaleColor001}
   &Drag(Drag=5, Use Linear Drag=true, Use Rotational Drag=false)
   &AccelerationForce(Acceleration=&UniformRangedVector(Minimum=<0 0 -25>, Maximum=<0 0 -75>))
   &SolveForcesAndVelocity(Clamp Velocity=false, Limit Acceleration=false)
   &ScaleSpriteSize(Scale Factor=&Vector2DFromFloat(Value=&FloatFromCurve(FloatCurve=DataInterface() {propertyValues {Curve {keys [
          {arriveTangent=5.2354, interpMode=RCIM_Cubic, leaveTangent=5.2354, tangentMode=RCTM_User, time=-0.0046, value=0.25}
          <at 0.1439 1.4923>
          <at 1.0011 3.4809 RCIM_Cubic>
         ]}}})))
   &DynamicMaterialParameters(
    Index 0 Param 1=&RandomRangeFloat(Minimum=0, Maximum=1)
    Index 0 Param 2=&FloatFromCurve(FloatCurve=DataInterface() {propertyValues {Curve {keys [
         {arriveTangent=0.0453, interpMode=RCIM_Cubic, tangentMode=RCTM_Break, time=0, value=-1}
         {arriveTangent=0.0604, interpMode=RCIM_Cubic, leaveTangent=0.0604, tangentMode=RCTM_User, time=1, value=-1}
        ]}}})
    Index 0 Param 3=&FloatFromCurve(FloatCurve=DataInterface() {propertyValues {Curve {keys [{interpMode=RCIM_Cubic, leaveTangent=4.6779, time=0.2851, value=0}]}}})
    Index 1 Param 1=&FloatFromCurve(
     FloatCurve=DataInterface() {propertyValues {Curve {keys [<at 0 0>,<at 0.25 1>]}}}
     Scale Curve=&RandomRangeFloat(Minimum=2, Maximum=1)
    )
    Index 1 Param 2=&FloatFromCurve(
     FloatCurve=DataInterface() {propertyValues {Curve {keys [
         {interpMode=RCIM_Cubic, leaveTangent=3.7422, tangentMode=RCTM_Break, time=0, value=0.2}
         {interpMode=RCIM_Cubic, leaveTangent=0.9066, time=1, value=2}
        ]}}}
     Scale Curve=&RandomRangeFloat(Minimum=0.5, Maximum=1)
    )
    Index 1 Param 3=&FloatFromCurve(FloatCurve=DataInterface() {propertyValues {Curve {keys [
         {interpMode=RCIM_Cubic, leaveTangent=2.8445, tangentMode=RCTM_Break, time=0, value=0.2}
         {interpMode=RCIM_Cubic, leaveTangent=63.2978, time=0.25, value=1}
        ]}}})
    Index 1 Param 4=&FloatFromCurve(
     FloatCurve=DataInterface() {propertyValues {Curve {keys [<at 0 1>,<at 1 0.8>]}}}
     Scale Curve=&RandomRangeFloat(Minimum=0.5, Maximum=1)
    )
    Index 2 Param 1=&FloatFromCurve(FloatCurve=DataInterface() {propertyValues {Curve {keys [{interpMode=RCIM_Cubic, leaveTangent=3.2309, tangentMode=RCTM_Break, time=0.1, value=0}, <at 1 1>]}}})
    Index 2 Param 2=&FloatFromCurve(
     FloatCurve=DataInterface() {propertyValues {Curve {keys [<at 0 0.2 RCIM_Cubic>,<at 1 0.8 RCIM_Cubic>]}}}
     Scale Curve=&RandomRangeFloat(Minimum=0.2, Maximum=1.2)
    )
    Index 2 Param 3=&FloatFromCurve(FloatCurve=DataInterface() {propertyValues {Curve {keys [
         {interpMode=RCIM_Cubic, tangentMode=RCTM_Break, time=0, value=1}
         {interpMode=RCIM_Cubic, leaveTangent=11.1233, tangentMode=RCTM_Break, time=1, value=2}
        ]}}})
    Index 2 Param 4=&FloatFromCurve(
     FloatCurve=DataInterface() {propertyValues {Curve {keys [<at 0 1 RCIM_Cubic>, {interpMode=RCIM_Cubic, leaveTangent=-1.2956, tangentMode=RCTM_Break, time=1, value=0.75}]}}}
     Scale Curve=0.8
    )
    Module.Param0WriteEnabled=true
    Module.Param0WriteEnabled1=true
    Module.Param0WriteEnabled2=true
    Module.Param1WriteEnabled=false
    Module.Param1WriteEnabled1=true
    Module.Param1WriteEnabled2=true
    Module.Param2WriteEnabled=true
    Module.Param2WriteEnabled1=true
    Module.Param2WriteEnabled2=true
    Module.Param3WriteEnabled=true
    Module.Param3WriteEnabled1=true
    Module.Param3WriteEnabled2=true
    Write Parameter Index 1=true
    Write Parameter Index 2=true
   )
  ]
  rendererClasses [&NiagaraSpriteRendererProperties]
  renderers [{rendererClass=&NiagaraSpriteRendererProperties}]
 }
 Sparkles {
  emitterUpdateScript [
   &EmitterState(Life Cycle Mode=::NewEnumerator1, Loop Behavior=::NewEnumerator1, Loop Duration=0.6, UseLoopDelay=true)
   &SpawnBurst_Instantaneous(Spawn Count=&RandomRangeInt(Minimum=25, Maximum=40), Spawn Time=0.05, Use Spawn Probability=false)
  ]
  particleSpawnScript [
   /SetVariables(Particles.VisibilityTag=&Add_Integer(
     A=&ModuloInt(A=&ReturnParticleIDAttributes(), B=2)
     B=&MakeIntFromFloat(Float=$User.UseGreenMat)
    )) {isSetParameters=true}
   &FN_SetSpawnGroup()
   &InitializeParticle(
    Lifetime=&Lerp_Float(
     A=&RandomRangeFloat(Minimum=0.25, Maximum=1.25)
     B=&RandomRangeFloat(Minimum=0.5, Maximum=1.75)
     Alpha=&MakeFloatFromInt(Int=&ModuloInt(A=&ReturnParticleIDAttributes(), B=2))
    )
    Module.Write Color=true
    Module.Write Lifetime=true
    Module.Write Mass=true
    Module.Write Material Random=true
    Module.Write Position=false
    Module.Write Scale=false
    Module.Write SpriteRotation=false
    Module.Write SpriteSize=true
    Sprite Rotation=&RandomRangeFloat(Minimum=0, Maximum=360)
    Sprite Size=&Vector2DFromFloat(Value=&Lerp_Float(
      A=&RandomRangeFloat(Minimum=3, Maximum=5)
      B=&RandomRangeFloat(Minimum=5, Maximum=12)
      Alpha=&MakeFloatFromInt(Int=&ModuloInt(A=&ReturnParticleIDAttributes(), B=2))
     ))
   )
   &SkeletalMeshLocation(
    Bone / Socket Index=$Particles.SpawnGroup
    Bone Sampling Mode=::NewEnumerator4
    Inherit Velocity=true
    Interpolate Mesh Sampling=false
    MaskSpawnGroup=false
    Skeletal Mesh=$User.SkeletalMesh
    VelocitySpeedLimit=false
   )
   /Niagara/Modules/Spawn/Location/SphereLocation(
    MaskSpawnGroup=false
    Module.Override Randomness=false
    Non Uniform Scale=<1 1 0.5>
    Sphere Origin:Vector3f=$Output.SkeletalMeshLocation.SampledPosition
    Sphere Radius=25
    Surface Only=false
   )
   &AddVelocity(Velocity=&Multiply_VectorByFloat(
     Vector=$Output.SphereLocation.SphereVector
     Float=&RandomRangeFloat(Minimum=1.5, Maximum=2, Evaluation Type=::NewEnumerator1)
    )) {module=AddVelocity001}
   &InheritVelocity(
    Module.Limit Speed=false
    Velocity Limit=0
    Velocity Scale=&VectorFromFloat(Value=&RandomRangeFloat(Minimum=0.3, Maximum=0.6, Evaluation Type=::NewEnumerator1))
   )
   &AddVelocity(Velocity=&RandomRangeVector(Minimum=<10 10 60>, Maximum=<20 20 100>)) {module=AddVelocity002}
  ]
  particleUpdateScript [
   &ParticleState()
   &ScaleSpriteSize(Scale Factor=&Vector2DFromFloat(Value=&FloatFromCurve(FloatCurve=DataInterface() {propertyValues {Curve {keys [
          <at 0 0 RCIM_Cubic>
          {arriveTangent=3.8197, interpMode=RCIM_Cubic, leaveTangent=3.8197, time=0.1034, value=1.1346}
          {arriveTangent=-0.4991, interpMode=RCIM_Cubic, leaveTangent=-0.4991, time=0.2059, value=0.7864}
          {arriveTangent=-0.9543, interpMode=RCIM_Cubic, leaveTangent=-0.9543, time=0.3019, value=1.0356}
          {arriveTangent=0.1149, interpMode=RCIM_Cubic, leaveTangent=0.1149, time=0.506, value=0.5}
          {arriveTangent=-0.8485, interpMode=RCIM_Cubic, leaveTangent=-0.8485, time=0.6432, value=1.0748}
          {arriveTangent=-3.012, interpMode=RCIM_Cubic, leaveTangent=-3.012, time=0.8021, value=0.2488}
          <at 1 0 RCIM_Cubic>
         ]}}})))
   &Drag(Drag=4.5, Use Linear Drag=true, Use Rotational Drag=false)
   &AccelerationForce(Acceleration=&RandomRangeVector(Minimum=<0 0 -35>, Maximum=<0 0 -60>))
   &CurlNoiseForce(
    Noise Strength=&FloatFromCurve(
     FloatCurve=DataInterface() {propertyValues {Curve {keys [
         <at 0 1 RCIM_Cubic>
         {arriveTangent=-1.1762, interpMode=RCIM_Cubic, leaveTangent=-1.1762, time=0.3044, value=0.8989}
         {arriveTangent=-1.2922, interpMode=RCIM_Cubic, leaveTangent=-1.2922, time=0.7087, value=0.1664}
         <at 1 0 RCIM_Cubic>
        ]}}}
     Scale Curve=100
    )
    Pan Noise=false
   )
   &SolveForcesAndVelocity(Clamp Velocity=false, Limit Acceleration=false)
   &Color(Color=&MultiplyLinearColor(
     Linear Color=&LerpLinearColors(
      StartColor=$User.Color_01
      EndColor=$User.Color_02
      LerpFactor=&FloatFromCurve(FloatCurve=DataInterface() {propertyValues {Curve {keys [<at 0.2 0>,<at 0.8 1>]}}})
     )
     Scale Factor=&MakeVector4_FromLinearColor(LinearColor=&RandomRangeLinearColor(Minimum=<rgba 1 1 1 1>, Maximum=<rgba 0.65 0.65 1 1>))
    )) {module=Color001}
   &ScaleColor(
    Color Value To Scale=$Particles.Color
    Scale Alpha=&FloatFromCurve(FloatCurve=DataInterface() {propertyValues {Curve {keys [<at 0 0 RCIM_Cubic>,<at 0.1 1 RCIM_Cubic>,<at 1 0>]}}})
    Scale RGB=&VectorFromFloat(Value=25)
    ScaleA=true
    ScaleRGB=true
   ) {module=ScaleColor001}
   &SpriteRotationRate(Rotation Rate=&Multiply_Float(
     A=&RandomRangeFloat(Minimum=-180, Maximum=180)
     B=&FloatFromCurve(FloatCurve=DataInterface() {propertyValues {Curve {keys [<at 0.35 0>,<at 1 1>]}}}, Scale Curve=1.25)
    ))
  ]
  rendererClasses [&NiagaraSpriteRendererProperties]
  renderers [
   {rendererClass=&NiagaraSpriteRendererProperties}
   {rendererClass=&NiagaraSpriteRendererProperties, rendererIndex=1}
   {rendererClass=&NiagaraSpriteRendererProperties, rendererIndex=2}
   {rendererClass=&NiagaraSpriteRendererProperties, rendererIndex=3}
  ]
 }
}
summary {
 systemName=NS_ScholarEmote_BookBurst
 emitters [
  {emitterName=Sparkles, rendererClasses [&NiagaraSpriteRendererProperties]}
  {emitterName=Smoke, rendererClasses [&NiagaraSpriteRendererProperties]}
 ]
 userVariables [
  User.Color_01 {type=LinearColor, description="", defaultValue {struct=LinearColor, value=<rgba 0.6445 0.1981 0.6514 1>}}
  User.Color_02 {type=LinearColor, description="", defaultValue {struct=LinearColor, value=<rgba 0.227 0.0953 0.6514 1>}}
  User.RibbonIDOffset {type=NiagaraInt32}
  User.SkeletalMesh {
   type=NiagaraDataInterfaceSkeletalMesh
   defaultValue {
    struct=NiagaraExt_VariableValue_DataInterface
    value {dataInterface=&self:NiagaraDataInterfaceSkeletalMesh_3}
   }
  }
  User.UseGreenMat {type=NiagaraFloat, defaultValue=0, description=""}
 ]
}
