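# Niagara explosion burst for the Scholar emote in front-end preview; every emitter spawns only when User.bIsFrontEndPreview is true.
# Embers, flipbook smoke and sparkles spawn around User.SkeletalMesh bones or sockets by spawn group, then drift with drag and acceleration.
# Shield decal sprites follow User.SKM_03 by particle ID, animated through dynamic material parameters.
# User colors tint everything, while sparkles set a VisibilityTag from ID parity plus UseGreenMat across four sprite renderers.
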NiagaraSystem /BRCosmetics/Effects/Fort_Effects/Emotes/Scholar/NS_ScholarEmote_Explosion_FE
dependencies {usedDataInterfaces [&NiagaraDataInterfaceCurve], usedDynamicInputs [], usedModules [&SystemState]}
emitters {
 BurstEmbers {
  emitterUpdateScript [
   &EmitterState(Life Cycle Mode=::NewEnumerator1, Loop Behavior=::NewEnumerator1, Loop Duration=0.4, UseLoopDelay=true)
   &SpawnBurst_Instantaneous(
    Spawn Count=&RandomRangeInt(Minimum=16, Maximum=32)
    Spawn Probability=&MakeCustomFloatFromBool(Bool=$User.bIsFrontEndPreview, True Float=1, False Float=0)
    Use Spawn Probability=true
   )
  ]
  particleSpawnScript [
   &FN_SetSpawnGroup(Number of Spawn Groups=17)
   /Niagara/Modules/Spawn/Initialization/InitializeParticle(
    Lifetime=&RandomRangeFloat(Minimum=0.5, Maximum=1, Evaluation Type=::NewEnumerator1)
    Module.Write Color=true
    Module.Write Lifetime=true
    Module.Write Mass=true
    Module.Write Material Random=true
    Module.Write Position=false
    Module.Write Scale=false
    Module.Write SpriteRotation=false
    Module.Write SpriteSize=true
    Sprite Size=&RandomRangeVector2D(Minimum=<0.5 1.25>, Maximum=<0.8 2.5>, Evaluation Type=::NewEnumerator1)
   )
   &AddVelocity(Velocity=&RandomRangeVector(Minimum=<-20 -20 -20>, Maximum=<20 20 20>, Evaluation Type=::NewEnumerator1))
   &SkeletalMeshLocation(
    Bone / Socket Index=$Particles.SpawnGroup
    Bone Sampling Mode=::NewEnumerator4
    Inherit Velocity=false
    MaskSpawnGroup=false
    Skeletal Mesh=$User.SkeletalMesh
   )
   /Niagara/Modules/Spawn/Location/SphereLocation(MaskSpawnGroup=false, Module.Override Randomness=false, Sphere Radius=13, Surface Only=false)
   &AddVelocity(Velocity=&Multiply_VectorByFloat(
     Vector=$Output.SphereLocation.SphereVector
     Float=&RandomRangeFloat(Minimum=10, Maximum=20, Evaluation Type=::NewEnumerator1)
    )) {module=AddVelocity001}
  ]
  particleUpdateScript [
   &ParticleState()
   &ScaleSpriteSize(Scale Factor=&Vector2DFromFloat(Value=&FloatFromCurve(FloatCurve=DataInterface() {propertyValues {Curve {keys [
          {arriveTangent=11.6538, interpMode=RCIM_Cubic, time=0, value=0.0211}
          {arriveTangent=0.3289, interpMode=RCIM_Cubic, leaveTangent=0.3289, time=0.125, value=1.5}
          <at 1 0.35 RCIM_Cubic>
         ]}}})))
   &Drag(Drag=4.5, Use Linear Drag=true, Use Rotational Drag=false)
   &AccelerationForce(Acceleration=<0 0 50>)
   &SolveForcesAndVelocity(Clamp Velocity=false, Limit Acceleration=false)
   &ScaleSpriteSizeBySpeed(Max Scale Factor=<1 4>, Source Velocity:Vector3f=&Multiply_Vector(A=<0 0.01 1>, B=$Particles.Velocity))
   &Color(Color=&MultiplyLinearColor(
     Linear Color=&LerpLinearColors(
      StartColor=$User.Color_01
      EndColor=$User.Color_02
      LerpFactor=&FloatFromCurve(FloatCurve=DataInterface() {propertyValues {Curve {keys [<at 0.2 0>,<at 0.8 1>]}}})
     )
     Scale Factor=&MakeVector4_FromLinearColor(LinearColor=&RandomRangeLinearColor(Minimum=<rgba 1 1 1 1>, Maximum=<rgba 0.65 0.65 1 1>))
    )) {module=Color002}
   &ScaleColor(
    Color Value To Scale=$Particles.Color
    Scale Alpha=&FloatFromCurve(FloatCurve=DataInterface() {propertyValues {Curve {keys [<at 0 0 RCIM_Cubic>,<at 0.1 1 RCIM_Cubic>,<at 1 0>]}}})
    Scale RGB=&VectorFromFloat(Value=6)
    ScaleA=true
    ScaleRGB=true
   ) {module=ScaleColor001}
  ]
  rendererClasses [&NiagaraSpriteRendererProperties]
  renderers [{rendererClass=&NiagaraSpriteRendererProperties}]
 }
 ShieldDecals {
  emitterUpdateScript [
   &EmitterState(Life Cycle Mode=::NewEnumerator1, Loop Behavior=::NewEnumerator1, Loop Duration=1, UseLoopDelay=true)
   &SpawnBurst_Instantaneous(
    Spawn Count=20
    Spawn Probability=&MakeCustomFloatFromBool(Bool=$User.bIsFrontEndPreview, True Float=1, False Float=0)
    Spawn Time=0.01
    Use Spawn Probability=true
   )
  ]
  particleSpawnScript [&InitializeParticle(
    Lifetime=0.95
    Module.Write Color=true
    Module.Write Lifetime=true
    Module.Write Mass=true
    Module.Write Material Random=true
    Module.Write Position=false
    Module.Write Scale=false
    Module.Write SpriteRotation=false
    Module.Write SpriteSize=true
    Sprite Size=&Vector2DFromFloat(Value=50)
   )]
  particleUpdateScript [
   &ParticleState()
   &Color(Color=&MultiplyLinearColor(
     Linear Color=$User.Color_01
     Scale Factor=&MakeFloat4FromVectorAndFloat("Vector (RGB)"=&VectorFromFloat(Value=3), "Float (Alpha)"=1)
    ))
   &ScaleColor(
    Color Value To Scale=$Particles.Color
    Scale Alpha=&FloatFromCurve(FloatCurve=DataInterface() {propertyValues {Curve {keys [{arriveTangent=7.9503, interpMode=RCIM_Cubic, time=0, value=0}, <at 0.2 1 RCIM_Cubic>, <at 1 0>]}}})
    Scale RGB=&VectorFromFloat(Value=2)
    ScaleA=true
    ScaleRGB=true
   )
   &SkeletalMeshLocation(
    Bone / Socket Index=&ReturnParticleIDAttributes()
    Bone Sampling Mode=::NewEnumerator4
    Inherit Velocity=false
    MaskSpawnGroup=false
    Skeletal Mesh=$User.SKM_03
   )
   &CameraOffset(Camera Offset Amount=25)
   &DynamicMaterialParameters(
    Index 0 Param 1=&FloatFromCurve(FloatCurve=DataInterface() {propertyValues {Curve {keys [
         <at 0 -0.5 RCIM_Cubic>
         {arriveTangent=10, interpMode=RCIM_Cubic, leaveTangent=10, time=0.0692, value=0.2738}
         {arriveTangent=5.0777, interpMode=RCIM_Cubic, leaveTangent=5.0777, time=0.15, value=1}
         {interpMode=RCIM_Cubic, leaveTangent=4.2839, time=1, value=5}
        ]}}})
    Index 0 Param 3=&FloatFromCurve(
     FloatCurve=DataInterface() {propertyValues {Curve {keys [{arriveTangent=2.3217, interpMode=RCIM_Cubic, time=0, value=-0.25}, {interpMode=RCIM_Cubic, leaveTangent=0.1844, time=1, value=1.4}]}}}
     Scale Curve=2
    )
    Index 0 Param 4=1
    Module.Param0WriteEnabled=true
    Module.Param1WriteEnabled=true
    Module.Param2WriteEnabled=true
    Module.Param3WriteEnabled=true
   )
  ]
  rendererClasses [&NiagaraSpriteRendererProperties]
  renderers [{rendererClass=&NiagaraSpriteRendererProperties}]
 }
 SmokeLoop {
  emitterUpdateScript [
   &EmitterState(Life Cycle Mode=::NewEnumerator1, Loop Behavior=::NewEnumerator1, Loop Duration=0.9, UseLoopDelay=true)
   &SpawnBurst_Instantaneous(
    Spawn Count=&RandomRangeInt(Minimum=6, Maximum=7)
    Spawn Probability=&MakeCustomFloatFromBool(Bool=$User.bIsFrontEndPreview, True Float=1, False Float=0)
    Spawn Time=0.01
    Use Spawn Probability=true
   )
  ]
  particleSpawnScript [
   &FN_SetSpawnGroup(Number of Spawn Groups=16)
   /SetVariables(Particles.SpriteRotationRate=&RandomRangeFloat(Minimum=-0.1, Maximum=-0.15, Evaluation Type=::NewEnumerator1)) {isSetParameters=true}
   &InitializeParticle(
    Lifetime=&RandomRangeFloat(Minimum=1.2, Maximum=1.5, Evaluation Type=::NewEnumerator1)
    Module.Write Color=true
    Module.Write Lifetime=true
    Module.Write Mass=true
    Module.Write Material Random=true
    Module.Write Position=false
    Module.Write Scale=false
    Module.Write SpriteRotation=true
    Module.Write SpriteSize=true
    Sprite Rotation=&NormalizedAngleToDegrees(NormalizedAngle=&RandomRangeFloat(Minimum=0, Maximum=1, Evaluation Type=::NewEnumerator1))
    Sprite Size=&Vector2DFromFloat(Value=&RandomRangeFloat(Minimum=40, Maximum=50, Evaluation Type=::NewEnumerator1))
   )
   &Color(Color=$User.Color_01)
   &CameraOffset(Camera Offset Amount=&RandomRangeFloat(Minimum=10, Maximum=20, Evaluation Type=::NewEnumerator1))
   &SkeletalMeshLocation(
    Bone / Socket Index=$Particles.SpawnGroup
    Bone Sampling Mode=::NewEnumerator4
    Inherit Velocity=true
    MaskSpawnGroup=false
    Skeletal Mesh=$User.SkeletalMesh
    VelocitySpeedLimit=false
   )
   &SphereLocation(MaskSpawnGroup=false, Module.Override Randomness=false, Sphere Radius=16, Surface Only=false)
   &AddVelocity(Velocity=&Multiply_VectorByFloat(Vector=$Output.SphereLocation.SphereVector, Float=1))
   &AddVelocity(Velocity=&RandomRangeVector(Minimum=<0 0 15>, Maximum=<0 0 25>, Evaluation Type=::NewEnumerator1)) {module=AddVelocity001}
  ]
  particleUpdateScript [
   &ParticleState()
   &ScaleColor(
    Scale Alpha=&FloatFromCurve(
     FloatCurve=DataInterface() {propertyValues {Curve {keys [
         {arriveTangent=3.4222, interpMode=RCIM_Cubic, time=0, value=0}
         <at 0.5 1 RCIM_Cubic>
         {interpMode=RCIM_Cubic, leaveTangent=-2.8405, time=1, value=0}
        ]}}}
     Scale Curve=0.1
    )
    Scale RGB=&VectorFromFloat(Value=&FloatFromCurve(FloatCurve=DataInterface() {propertyValues {Curve {keys [{arriveTangent=-2.1797, interpMode=RCIM_Cubic, time=0, value=1}, <at 1 0.0987>]}}}))
    ScaleA=true
    ScaleRGB=true
   )
   /Niagara/Modules/Update/SubUV/SubUVAnimation(Number Of Frames=30)
   &SpriteRotationRate(Rotation Rate=&NormalizedAngleToDegrees(NormalizedAngle=$Particles.SpriteRotationRate))
   &ScaleSpriteSize(Scale Factor=&Vector2DFromFloat(Value=&FloatFromCurve(FloatCurve=DataInterface() {propertyValues {Curve {keys [{arriveTangent=1.9916, interpMode=RCIM_Cubic, time=0, value=0.5}, <at 1 1.5 RCIM_Cubic>]}}})))
   &AccelerationForce(Acceleration=<0 0 20>)
   &Drag(
    Drag=&RandomRangeFloat(Minimum=1, Maximum=1.5, Evaluation Type=::NewEnumerator1)
    Use Linear Drag=true
    Use Rotational Drag=false
   )
   &SolveForcesAndVelocity(Clamp Velocity=false, Limit Acceleration=false)
  ]
  rendererClasses [&NiagaraSpriteRendererProperties]
  renderers [{rendererClass=&NiagaraSpriteRendererProperties}]
 }
 Sparkles {
  emitterUpdateScript [
   &EmitterState(Life Cycle Mode=::NewEnumerator1, Loop Behavior=::NewEnumerator1, Loop Duration=0.6, UseLoopDelay=true)
   &SpawnBurst_Instantaneous(
    Spawn Count=&RandomRangeInt(Minimum=100, Maximum=120)
    Spawn Probability=&MakeCustomFloatFromBool(Bool=$User.bIsFrontEndPreview, True Float=1, False Float=0)
    Spawn Time=0.05
    Use Spawn Probability=true
   )
  ]
  particleSpawnScript [
   /SetVariables(Particles.VisibilityTag=&Add_Integer(
     A=&ModuloInt(A=&ReturnParticleIDAttributes(), B=2)
     B=&MakeIntFromFloat(Float=$User.UseGreenMat)
    )) {isSetParameters=true}
   &FN_SetSpawnGroup(Number of Spawn Groups=17)
   &InitializeParticle(
    Lifetime=&Lerp_Float(
     A=&RandomRangeFloat(Minimum=0.25, Maximum=2)
     B=&RandomRangeFloat(Minimum=0.5, Maximum=2.5)
     Alpha=&MakeFloatFromInt(Int=&ModuloInt(A=&ReturnParticleIDAttributes(), B=2))
    )
    Module.Write Color=true
    Module.Write Lifetime=true
    Module.Write Mass=true
    Module.Write Material Random=true
    Module.Write Position=false
    Module.Write Scale=false
    Module.Write SpriteRotation=false
    Module.Write SpriteSize=true
    Sprite Rotation=&RandomRangeFloat(Minimum=0, Maximum=360)
    Sprite Size=&Vector2DFromFloat(Value=&Lerp_Float(
      A=&RandomRangeFloat(Minimum=3, Maximum=5)
      B=&RandomRangeFloat(Minimum=5, Maximum=12)
      Alpha=&MakeFloatFromInt(Int=&ModuloInt(A=&ReturnParticleIDAttributes(), B=2))
     ))
   )
   &SkeletalMeshLocation(
    Bone / Socket Index=$Particles.SpawnGroup
    Bone Sampling Mode=::NewEnumerator4
    Inherit Velocity=true
    MaskSpawnGroup=false
    Skeletal Mesh=$User.SkeletalMesh
    VelocitySpeedLimit=false
   )
   &SphereLocation(MaskSpawnGroup=false, Module.Override Randomness=false, Sphere Radius=15, Surface Only=false)
   &AddVelocity(Velocity=&Multiply_VectorByFloat(
     Vector=$Output.SphereLocation.SphereVector
     Float=&RandomRangeFloat(Minimum=13, Maximum=20, Evaluation Type=::NewEnumerator1)
    )) {module=AddVelocity001}
   &InheritVelocity(
    Module.Limit Speed=false
    Velocity Limit=0
    Velocity Scale=&VectorFromFloat(Value=&RandomRangeFloat(Minimum=0.5, Maximum=1.5, Evaluation Type=::NewEnumerator1))
   )
   &AddVelocity(Velocity=&RandomRangeVector(Minimum=<0 0 40>, Maximum=<1 0 120>)) {module=AddVelocity002}
  ]
  particleUpdateScript [
   &ParticleState()
   &ScaleSpriteSize(Scale Factor=&Vector2DFromFloat(Value=&FloatFromCurve(FloatCurve=DataInterface() {propertyValues {Curve {keys [
          <at 0 0 RCIM_Cubic>
          {arriveTangent=3.8197, interpMode=RCIM_Cubic, leaveTangent=3.8197, time=0.1034, value=1.1346}
          {arriveTangent=-0.4991, interpMode=RCIM_Cubic, leaveTangent=-0.4991, time=0.2059, value=0.7864}
          {arriveTangent=-0.9543, interpMode=RCIM_Cubic, leaveTangent=-0.9543, time=0.3019, value=1.0356}
          {arriveTangent=0.1149, interpMode=RCIM_Cubic, leaveTangent=0.1149, time=0.506, value=0.5}
          {arriveTangent=-0.8485, interpMode=RCIM_Cubic, leaveTangent=-0.8485, time=0.6432, value=1.0748}
          {arriveTangent=-3.012, interpMode=RCIM_Cubic, leaveTangent=-3.012, time=0.8021, value=0.2488}
          <at 1 0 RCIM_Cubic>
         ]}}})))
   &Drag(Drag=4.5, Use Linear Drag=true, Use Rotational Drag=false)
   &AccelerationForce(Acceleration=&RandomRangeVector(Minimum=<0 0 -50>, Maximum=<0 0 -100>))
   &SolveForcesAndVelocity(Clamp Velocity=false, Limit Acceleration=false)
   &Color(Color=&MultiplyLinearColor(
     Linear Color=&LerpLinearColors(
      StartColor=$User.Color_01
      EndColor=$User.Color_02
      LerpFactor=&FloatFromCurve(FloatCurve=DataInterface() {propertyValues {Curve {keys [<at 0.2 0>,<at 0.8 1>]}}})
     )
     Scale Factor=&MakeVector4_FromLinearColor(LinearColor=&RandomRangeLinearColor(Minimum=<rgba 1 1 1 1>, Maximum=<rgba 0.65 0.65 1 1>))
    )) {module=Color001}
   &ScaleColor(
    Color Value To Scale=$Particles.Color
    Scale Alpha=&FloatFromCurve(FloatCurve=DataInterface() {propertyValues {Curve {keys [<at 0 0 RCIM_Cubic>,<at 0.1 1 RCIM_Cubic>,<at 1 0>]}}})
    Scale RGB=&VectorFromFloat(Value=25)
    ScaleA=true
    ScaleRGB=true
   ) {module=ScaleColor001}
   &SpriteRotationRate(Rotation Rate=&Multiply_Float(
     A=&RandomRangeFloat(Minimum=-180, Maximum=180)
     B=&FloatFromCurve(FloatCurve=DataInterface() {propertyValues {Curve {keys [<at 0.35 0>,<at 1 1>]}}}, Scale Curve=1.25)
    ))
  ]
  rendererClasses [&NiagaraSpriteRendererProperties]
  renderers [
   {rendererClass=&NiagaraSpriteRendererProperties}
   {rendererClass=&NiagaraSpriteRendererProperties, rendererIndex=1}
   {rendererClass=&NiagaraSpriteRendererProperties, rendererIndex=2}
   {rendererClass=&NiagaraSpriteRendererProperties, rendererIndex=3}
  ]
 }
}
summary {
 systemName=NS_ScholarEmote_Explosion_FE
 emitters [
  {emitterName=BurstEmbers, rendererClasses [&NiagaraSpriteRendererProperties]}
  {emitterName=SmokeLoop, rendererClasses [&NiagaraSpriteRendererProperties]}
  {emitterName=ShieldDecals, rendererClasses [&NiagaraSpriteRendererProperties]}
  {emitterName=Sparkles, rendererClasses [&NiagaraSpriteRendererProperties]}
 ]
 userVariables [
  User.bIsFrontEndPreview {type=NiagaraBool}
  User.Color_01 {type=LinearColor, description="", defaultValue {struct=LinearColor, value=<rgba 0.6445 0.1981 0.6514 1>}}
  User.Color_02 {type=LinearColor, description="", defaultValue {struct=LinearColor, value=<rgba 0.227 0.0953 0.6514 1>}}
  User.RibbonIDOffset {type=NiagaraInt32}
  User.SkeletalMesh {
   type=NiagaraDataInterfaceSkeletalMesh
   defaultValue {
    struct=NiagaraExt_VariableValue_DataInterface
    value {dataInterface=&self:NiagaraDataInterfaceSkeletalMesh_0}
   }
  }
  User.SKM_02 {
   type=NiagaraDataInterfaceSkeletalMesh
   defaultValue {
    struct=NiagaraExt_VariableValue_DataInterface
    value {dataInterface=&self:NiagaraDataInterfaceSkeletalMesh_1}
   }
  }
  User.SKM_03 {
   type=NiagaraDataInterfaceSkeletalMesh
   defaultValue {
    struct=NiagaraExt_VariableValue_DataInterface
    value {dataInterface=&self:NiagaraDataInterfaceSkeletalMesh_2}
   }
  }
  User.UseGreenMat {type=NiagaraFloat, defaultValue=0, description=""}
 ]
}
