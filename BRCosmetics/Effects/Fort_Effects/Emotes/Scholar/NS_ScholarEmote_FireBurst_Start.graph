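# Niagara system for the Scholar emote's fire burst start, placing recolorable magical flames (magenta-to-purple by default) at bones or sockets of User.SkeletalMesh.
# Fire_Loop emits rising SubUV flame sprites, tinted from User.Color_01 toward Color_02.
# Sparks adds curl-noise, speed-stretched sprites in the same colors, with curves timing a brief spawn spike and upward kick early in the loop.
# FireAirglow spawns one short-lived Color_01 glow sprite and light with curve-flickered alpha.
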NiagaraSystem /BRCosmetics/Effects/Fort_Effects/Emotes/Scholar/NS_ScholarEmote_FireBurst_Start
dependencies {usedDataInterfaces [&NiagaraDataInterfaceCurve], usedDynamicInputs [], usedModules [&SystemState]}
emitters {
 FireAirglow {
  emitterUpdateScript [
   &EmitterState()
   &SpawnBurst_Instantaneous(Spawn Count=1, Spawn Probability=HlslExpression() {hlslExpression="Emitter.LoopCount < 1"}, Use Spawn Probability=true)
  ]
  particleSpawnScript [/Niagara/Modules/Spawn/Initialization/InitializeParticle(
    Color=<rgba 0.6445 0.1981 0.6514 1>
    Lifetime=0.9
    Module.Write Color=true
    Module.Write Lifetime=true
    Module.Write Mass=true
    Module.Write Material Random=true
    Module.Write Position=false
    Module.Write Scale=false
    Module.Write SpriteRotation=false
    Module.Write SpriteSize=true
    Sprite Size=&Multiply_Vector2(A=<50 50>, B=&Vector2DFromFloat(Value=0.9))
   )]
  particleUpdateScript [
   &ParticleState()
   &SkeletalMeshLocation(
    Apply Sampled Orientation to Mesh Particles=true
    Bone / Socket Index=2
    Bone Sampling Mode=::NewEnumerator4
    Inherit Velocity=false
    MaskSpawnGroup=false
    Sampled Orientation Coordinate Space=::Local
    Sampled Position Coordinate Space=::Local
    Sampled Position Offset=<-5 0 10>
    Sampled Position Offset Coordinate Space:ENiagaraMeshTransforms=::NewEnumerator2
    Sampled Velocity Coordinate Space:ENiagaraCoordinateSpace=::Local
    Skeletal Mesh=$User.SkeletalMesh
   ) {module=SkeletalMeshLocation002}
   &Color(Color=$User.Color_01)
   /SetVariables(
    Particles.MeshColor:LinearColor=&MultiplyLinearColor(Linear Color=$Particles.Color, Scale Factor=<1 1 1 1>)
    Particles.SpriteColor=&MultiplyLinearColor(
     Linear Color=$Particles.Color
     Scale Factor=&MakeFloat4FromVectorAndFloat(
      "Vector (RGB)"=<1 1 1>
      "Float (Alpha)"=&FloatFromCurve(
       FloatCurve=DataInterface() {propertyValues {Curve {keys [
           <at 0 0 RCIM_Cubic>
           {arriveTangent=3.5783, interpMode=RCIM_Cubic, leaveTangent=3.5783, time=0.0968, value=0.9921}
           {arriveTangent=1.1135, interpMode=RCIM_Cubic, leaveTangent=1.1135, time=0.2116, value=0.7573}
           {arriveTangent=0.0266, interpMode=RCIM_Cubic, leaveTangent=0.0266, time=0.3435, value=1.2667}
           {arriveTangent=-0.1747, interpMode=RCIM_Cubic, leaveTangent=-0.1747, time=0.5441, value=0.7662}
           {arriveTangent=2.9922, interpMode=RCIM_Cubic, leaveTangent=2.9922, time=0.6139, value=1.2195}
           {arriveTangent=-1.5511, interpMode=RCIM_Cubic, leaveTangent=-1.5511, time=0.722, value=1.2984}
           {arriveTangent=-4.6703, interpMode=RCIM_Cubic, leaveTangent=-4.6703, time=0.8433, value=0.8636}
           <at 1 0 RCIM_Cubic>
          ]}}}
       Scale Curve=0.015
      )
     )
    )
   ) {isSetParameters=true}
   &DynamicMaterialParameters(
    Index 0 Param 2=1
    Index 0 Param 3=12
    Module.Param0WriteEnabled=true
    Module.Param1WriteEnabled=true
    Module.Param2WriteEnabled=true
    Module.Param3WriteEnabled=true
   )
   &FN_Light_Attributes(
    Color=$Particles.SpriteColor
    Create Color Attribute=true
    Create Exponent Attribute=true
    Create Radius Attribute=true
    Exponent=1
    Radius=25
    Scale Color=2
   )
  ]
  rendererClasses [&NiagaraSpriteRendererProperties, &NiagaraLightRendererProperties]
  renderers [
   {rendererClass=&NiagaraSpriteRendererProperties}
   {rendererClass=&NiagaraLightRendererProperties, rendererIndex=1}
  ]
 }
 Fire_Loop {
  emitterUpdateScript [
   &EmitterState()
   &SpawnRate(
    Spawn Probability=&FloatFromCurve(FloatCurve=DataInterface() {propertyValues {Curve {keys [<at 0 1>,<at 1 0.65>]}}}, CurveIndex=$Emitter.NormalizedLoopAge)
    SpawnRate=20
    Use Spawn Probability=false
   )
  ]
  particleSpawnScript [
   &InitializeParticle(
    Color=&RandomRangeLinearColor(Minimum=<rgba 0 0 0 1>, Maximum=<rgba 1 1 1 1>)
    Lifetime=&UniformRangedFloat(Minimum=0.9, Maximum=1)
    Module.Write Color=false
    Module.Write Lifetime=true
    Module.Write Mass=false
    Module.Write Material Random=true
    Module.Write Position=false
    Module.Write Scale=false
    Module.Write SpriteRotation=true
    Module.Write SpriteSize=true
    Position=&SimulationPosition()
    Sprite Rotation=&UniformRangedFloat(Minimum=-180, Maximum=180)
    Sprite Size=&Vector2DFromFloat(Value=&RandomRangeFloat(Minimum=23, Maximum=27))
   )
   &FN_SetSpawnGroup()
   &SkeletalMeshLocation(
    Apply Sampled Orientation to Mesh Particles=true
    Bone / Socket Index=$Particles.SpawnGroup
    Bone Sampling Mode=::NewEnumerator4
    Inherit Velocity=true
    Inherited Velocity Speed Limit=30
    MaskSpawnGroup=false
    Sampled Orientation Coordinate Space=::Local
    Sampled Position Coordinate Space=::Local
    Sampled Position Offset=&RandomRangeVector(Minimum=<0 -10 -10>, Maximum=<0 10 10>)
    Sampled Position Offset Coordinate Space:ENiagaraMeshTransforms=::NewEnumerator3
    Sampled Velocity Coordinate Space:ENiagaraCoordinateSpace=::Local
    Skeletal Mesh=$User.SkeletalMesh
    VelocitySpeedLimit=true
   ) {module=SkeletalMeshLocation002}
   &CameraOffset(Camera Offset Amount=&UniformRangedFloat(Minimum=2.5, Maximum=5))
   &AddVelocity(Velocity=&RandomRangeVector(Minimum=<-2 -2 25>, Maximum=<2 2 40>))
  ]
  particleUpdateScript [
   &ParticleState()
   &Color(Color=&LerpLinearColors(
     StartColor=<rgba 1 1 1 1>
     EndColor=<rgba 0.7 0.7 0.7 1>
     LerpFactor=&RandomRangeFloat(Minimum=0, Maximum=1)
    ))
   &ScaleColor(
    Color Value To Scale=$Particles.Color
    Scale Alpha=&FloatFromCurve(FloatCurve=DataInterface() {propertyValues {Curve {keys [
         <at 0 0 RCIM_Cubic>
         {arriveTangent=0.2067, interpMode=RCIM_Cubic, leaveTangent=0.2067, tangentMode=RCTM_User, time=0.2, value=1}
         {arriveTangent=-1.3652, interpMode=RCIM_Cubic, leaveTangent=-1.3652, tangentMode=RCTM_User, time=0.8003, value=0}
        ]}}})
    Scale RGB=&MakeVectorFromLinearColorRGB(Color=&LerpLinearColors(
      StartColor=$User.Color_01
      EndColor=$User.Color_02
      LerpFactor=&FloatFromCurve(FloatCurve=DataInterface() {propertyValues {Curve {keys [<at 0 0>,<at 0.9 0.5>]}}})
     ))
    ScaleA=true
    ScaleRGB=true
   )
   &ScaleColor(Scale Alpha=1.25, Scale RGB=&VectorFromFloat(Value=2.5), ScaleA=true, ScaleRGB=true) {module=ScaleColor002}
   &AccelerationForce(Acceleration=&RandomRangeVector(Minimum=<0 0 20>, Maximum=<0 0 20>))
   &Drag(Drag=2, Use Linear Drag=true, Use Rotational Drag=false)
   &SolveForcesAndVelocity(Clamp Velocity=false, Limit Acceleration=false)
   /Niagara/Modules/Update/SubUV/SubUVAnimation(Number Of Frames=64)
   &ScaleSpriteSize(Scale Factor=&Vector2DFromFloat(Value=&FloatFromCurve(FloatCurve=DataInterface() {propertyValues {Curve {keys [
          {arriveTangent=-0.0625, interpMode=RCIM_Cubic, leaveTangent=-0.0625, tangentMode=RCTM_User, time=0, value=0}
          <at 0.4 1>
          {arriveTangent=-0.3333, interpMode=RCIM_Cubic, leaveTangent=-0.3333, time=0.6722, value=1}
          <at 1 0.8 RCIM_Cubic>
         ]}}})))
   &DynamicMaterialParameters(
    Index 0 Param 1=&FloatFromCurve(FloatCurve=DataInterface() {propertyValues {Curve {keys [{arriveTangent=2.7551, interpMode=RCIM_Cubic, leaveTangent=2.7551, tangentMode=RCTM_User, time=0, value=0.25}, <at 1 1>]}}})
    Index 0 Param 2=1
    Index 0 Param 3=1.5
    Index 0 Param 4=1.25
    Module.Param0WriteEnabled=true
    Module.Param1WriteEnabled=true
    Module.Param2WriteEnabled=true
    Module.Param3WriteEnabled=true
   ) {module=DynamicMaterialParameters001}
  ]
  rendererClasses [&NiagaraSpriteRendererProperties]
  renderers [{rendererClass=&NiagaraSpriteRendererProperties}]
 }
 Sparks {
  emitterUpdateScript [
   &EmitterState()
   &SpawnRate(
    SpawnRate=&FloatFromCurve(
     FloatCurve=DataInterface() {propertyValues {Curve {keys [<at 0 1>,<at 0.2 1>,<at 0.25 2>,<at 0.3 1>,<at 1 1>]}}}
     CurveIndex=$Emitter.NormalizedLoopAge
     Scale Curve=40
    )
    Use Spawn Probability=false
   )
  ]
  particleSpawnScript [
   &InitializeParticle(
    Color=<rgba 1 2 3 1>
    Lifetime=&RandomRangeFloat(Minimum=0.6, Maximum=0.8)
    Module.Write Color=false
    Module.Write Lifetime=true
    Module.Write Mass=true
    Module.Write Material Random=true
    Module.Write Position=false
    Module.Write Scale=false
    Module.Write SpriteRotation=false
    Module.Write SpriteSize=true
    Position=&SimulationPosition()
    Sprite Size=&RandomRangeVector2D(Minimum=<0.8 1.5>, Maximum=<1.1 2>)
   )
   &FN_SetSpawnGroup()
   &SkeletalMeshLocation(
    Apply Sampled Orientation to Mesh Particles=true
    Bone / Socket Index=$Particles.SpawnGroup
    Bone Sampling Mode=::NewEnumerator4
    Inherit Velocity=false
    MaskSpawnGroup=false
    Sampled Orientation Coordinate Space=::Local
    Sampled Position Coordinate Space=::Local
    Sampled Position Offset=&RandomRangeVector(Minimum=<0 -8 -8>, Maximum=<0 8 8>)
    Sampled Position Offset Coordinate Space:ENiagaraMeshTransforms=::NewEnumerator3
    Sampled Velocity Coordinate Space:ENiagaraCoordinateSpace=::Local
    Skeletal Mesh=$User.SkeletalMesh
   ) {module=SkeletalMeshLocation002}
   &AddVelocity(Velocity=&RandomRangeVector(Minimum=<-35 -35 20>, Maximum=<35 35 30>))
   &CameraOffset(Camera Offset Amount=2.5)
   &AddVelocity(Velocity=&MakeVector(
     X=0
     Y=0
     Z=&FloatFromCurve(
      FloatCurve=DataInterface() {propertyValues {Curve {keys [<at 0 0>,<at 0.25 0>,<at 0.3 1>,<at 0.35 0>,<at 1 0>]}}}
      CurveIndex=$Emitter.NormalizedLoopAge
      Scale Curve=80
     )
    )) {module=AddVelocity001}
  ]
  particleUpdateScript [
   &ParticleState()
   &Drag(Use Linear Drag=true, Use Rotational Drag=false)
   &AccelerationForce(Acceleration=<0 0 25>)
   &CurlNoiseForce(Noise Frequency=100, Noise Strength=&RandomRangeFloat(Minimum=300, Maximum=500), Pan Noise=false)
   &SolveForcesAndVelocity(Clamp Velocity=false, Limit Acceleration=false)
   &ScaleSpriteSizeBySpeed(Max Scale Factor=<0.5 10>, Min Scale Factor=<0.5 0.5>)
   &ScaleSpriteSize(Scale Factor=&Vector2DFromFloat(Value=&FloatFromCurve(FloatCurve=DataInterface() {propertyValues {Curve {keys [{arriveTangent=-21.3532, interpMode=RCIM_Cubic, leaveTangent=-6.002, tangentMode=RCTM_User, time=0.5, value=1}, <at 1 0 RCIM_Cubic>]}}})))
   &Color(Color=&LerpLinearColors(
     StartColor=$User.Color_01
     EndColor=$User.Color_02
     LerpFactor=&FloatFromCurve(FloatCurve=DataInterface() {propertyValues {Curve {keys [<at 0 0>,<at 1 0.5>]}}})
    ))
   &ScaleColor(
    Color Value To Scale=$Particles.Color
    Scale Alpha=&FloatFromCurve(FloatCurve=DataInterface() {propertyValues {Curve {keys [<at 0 0 RCIM_Cubic>,<at 0.1 1 RCIM_Cubic>,<at 1 0>]}}})
    Scale RGB=&VectorFromFloat(Value=6)
    ScaleA=true
    ScaleRGB=true
   )
  ]
  rendererClasses [&NiagaraSpriteRendererProperties]
  renderers [{rendererClass=&NiagaraSpriteRendererProperties}]
 }
}
summary {
 systemName=NS_ScholarEmote_FireBurst_Start
 emitters [
  {emitterName=Fire_Loop, rendererClasses [&NiagaraSpriteRendererProperties]}
  {emitterName=Sparks, rendererClasses [&NiagaraSpriteRendererProperties]}
  {emitterName=FireAirglow, rendererClasses [&NiagaraSpriteRendererProperties, &NiagaraLightRendererProperties]}
 ]
 userVariables [
  User.Color_01 {type=LinearColor, description="", defaultValue {struct=LinearColor, value=<rgba 0.7991 0.0103 0.7991 1>}}
  User.Color_02 {type=LinearColor, description="", defaultValue {struct=LinearColor, value=<rgba 0.227 0.0953 0.6514 1>}}
  User.RenderVis {type=NiagaraFloat, defaultValue=0, description=""}
  User.RibbonIDOffset {type=NiagaraInt32}
  User.SkeletalMesh {
   type=NiagaraDataInterfaceSkeletalMesh
   defaultValue {
    struct=NiagaraExt_VariableValue_DataInterface
    value {dataInterface=&self:NiagaraDataInterfaceSkeletalMesh_3}
   }
  }
  User.UseGreenMat {type=NiagaraFloat, defaultValue=0, description=""}
 ]
}
