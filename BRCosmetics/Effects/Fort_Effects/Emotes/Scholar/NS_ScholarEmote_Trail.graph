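# Niagara trail effect for the Scholar emote, with three emitters that sample locations on the User.SKM_02 skeletal mesh and take tint from user colors.
# Halo is one mesh-following sprite flash, and RIbbon is a short-lived trail with brighter, wider secondary attributes for its two renderers.
# Sparkles drift, flicker in size and blend between both colors.
# Particle ID parity varies their size and lifetime and, offset by UseGreenMat, sets a visibility tag across four renderers.
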NiagaraSystem /BRCosmetics/Effects/Fort_Effects/Emotes/Scholar/NS_ScholarEmote_Trail
dependencies {usedDataInterfaces [&NiagaraDataInterfaceCurve], usedDynamicInputs [], usedModules [&SystemState]}
emitters {
 Halo {
  emitterUpdateScript [
   &EmitterState()
   &SpawnBurst_Instantaneous(Spawn Count=1, Spawn Time=0.01, Use Spawn Probability=false)
  ]
  particleSpawnScript [&InitializeParticle(Lifetime=2.5, Sprite Size Mode=::NewEnumerator3, UsePositionOffset=false, Write Material Random=true)]
  particleUpdateScript [
   &ParticleState()
   &SkeletalMeshLocation(Bone Sampling Mode=::NewEnumerator4, Inherit Velocity=false, MaskSpawnGroup=false, Skeletal Mesh=$User.SKM_02)
   &Color(Color=&MultiplyLinearColor(
     Linear Color=$User.Color_01
     Scale Factor=&MakeFloat4FromVectorAndFloat(
      "Vector (RGB)"=&VectorFromFloat(Value=&FloatFromCurve(
        FloatCurve=DataInterface() {propertyValues {Curve {keys [{arriveTangent=7.9503, interpMode=RCIM_Cubic, time=0, value=0}, <at 0.1 3 RCIM_Cubic>, <at 1 0>]}}}
        Scale Curve=4
       ))
      "Float (Alpha)"=1
     )
    )) {module=Color001}
   &ScaleColor(
    Color Value To Scale=$Particles.Color
    Scale Alpha=&FloatFromCurve(
     FloatCurve=DataInterface() {propertyValues {Curve {keys [
         {arriveTangent=7.9503, interpMode=RCIM_Cubic, time=0, value=0}
         {arriveTangent=0.2271, interpMode=RCIM_Cubic, leaveTangent=0.2271, tangentMode=RCTM_User, time=0.1, value=1}
         {arriveTangent=-1.0554, interpMode=RCIM_Cubic, leaveTangent=-1.0554, tangentMode=RCTM_User, time=0.8564, value=0.8745}
         <at 1 0>
        ]}}}
     Scale Curve=2
    )
    Scale RGB=&VectorFromFloat(Value=&FloatFromCurve(FloatCurve=DataInterface() {propertyValues {Curve {keys [{arriveTangent=7.9503, interpMode=RCIM_Cubic, time=0, value=0}, <at 0.2 1 RCIM_Cubic>, <at 1 0>]}}}))
    ScaleA=true
    ScaleRGB=false
   )
   &ScaleSpriteSize(Scale Factor=&Vector2DFromFloat(Value=&FloatFromCurve(
      FloatCurve=DataInterface() {propertyValues {Curve {keys [
          {arriveTangent=5.7712, interpMode=RCIM_Cubic, leaveTangent=5.7712, tangentMode=RCTM_User, time=0, value=0}
          {arriveTangent=1.25, interpMode=RCIM_Cubic, leaveTangent=1.25, time=0.35, value=1}
          <at 1 1.25>
         ]}}}
      Scale Curve=1.5
     )))
   &DynamicMaterialParameters(
    Index 0 Param 1=2
    Index 0 Param 2=1
    Index 0 Param 3=1
    Module.Param0WriteEnabled=true
    Module.Param1WriteEnabled=true
    Module.Param2WriteEnabled=true
    Module.Param3WriteEnabled=true
   )
  ]
  rendererClasses [&NiagaraSpriteRendererProperties]
  renderers [{rendererClass=&NiagaraSpriteRendererProperties}]
 }
 RIbbon {
  emitterUpdateScript [
   &EmitterState()
   &SpawnRate(SpawnRate=120, Use Spawn Probability=false)
  ]
  particleSpawnScript [
   &InitializeRibbon(
    Lifetime=0.15
    Module.Write Color=true
    Module.Write Lifetime=true
    Module.Write Mass=true
    Module.Write Material Random=true
    Module.Write Position=false
    Module.Write RibbonTwist=false
    Module.Write RibbonWidth=true
    Ribbon Width=2
   )
   &SkeletalMeshLocation(Bone Sampling Mode=::NewEnumerator4, Inherit Velocity=false, MaskSpawnGroup=false, Skeletal Mesh=$User.SKM_02)
  ]
  particleUpdateScript [
   &ParticleState()
   &ScaleRibbonWidth(Ribbon Width Scale=&Multiply_Float(
     A=&FloatFromCurve(FloatCurve=DataInterface() {propertyValues {Curve {keys [<at 0 0 RCIM_Cubic>,<at 0.15 1 RCIM_Cubic>,<at 1 0 RCIM_Cubic>]}}}, Scale Curve=1.5)
     B=&FloatFromCurve(
      FloatCurve=DataInterface() {propertyValues {Curve {keys [
          <at 0 0 RCIM_Cubic>
          {arriveTangent=0.35, interpMode=RCIM_Cubic, leaveTangent=0.35, time=0.1, value=1}
          {arriveTangent=-0.7222, interpMode=RCIM_Cubic, leaveTangent=-0.7222, time=0.7, value=1}
          <at 1 0.35 RCIM_Cubic>
         ]}}}
      CurveIndex=$Emitter.NormalizedLoopAge
      Scale Curve=0.6
     )
    ))
   &Color(Color=$User.Color_01)
   /SetVariables(
    Particles.Color_02:LinearColor=&MultiplyLinearColor(
     Linear Color=$Particles.Color
     Scale Factor=&MakeFloat4FromVectorAndFloat("Vector (RGB)"=&VectorFromFloat(Value=2), "Float (Alpha)"=1)
    )
    Particles.RibbonWidth_02:NiagaraFloat=&Multiply_Float(A=$Particles.RibbonWidth, B=6)
   ) {isSetParameters=true}
  ]
  rendererClasses [&NiagaraRibbonRendererProperties]
  renderers [
   {rendererClass=&NiagaraRibbonRendererProperties}
   {rendererClass=&NiagaraRibbonRendererProperties, rendererIndex=1}
  ]
 }
 Sparkles {
  emitterUpdateScript [
   &EmitterState()
   &SpawnRate(SpawnRate=20, Use Spawn Probability=false)
  ]
  particleSpawnScript [
   /SetVariables(Particles.VisibilityTag=&Add_Integer(
     A=&ModuloInt(A=&ReturnParticleIDAttributes(), B=2)
     B=&MakeIntFromFloat(Float=$User.UseGreenMat)
    )) {isSetParameters=true}
   &FN_SetSpawnGroup(Number of Spawn Groups=17)
   /Niagara/Modules/Spawn/Initialization/InitializeParticle(
    Lifetime=&Lerp_Float(
     A=&RandomRangeFloat(Minimum=0.25, Maximum=1)
     B=&RandomRangeFloat(Minimum=0.5, Maximum=1.5)
     Alpha=&MakeFloatFromInt(Int=&ModuloInt(A=&ReturnParticleIDAttributes(), B=2))
    )
    Module.Write Color=true
    Module.Write Lifetime=true
    Module.Write Mass=true
    Module.Write Material Random=true
    Module.Write Position=false
    Module.Write Scale=false
    Module.Write SpriteRotation=false
    Module.Write SpriteSize=true
    Sprite Rotation=&RandomRangeFloat(Minimum=0, Maximum=360)
    Sprite Size=&Vector2DFromFloat(Value=&Lerp_Float(
      A=&RandomRangeFloat(Minimum=3, Maximum=4)
      B=&RandomRangeFloat(Minimum=5, Maximum=6)
      Alpha=&MakeFloatFromInt(Int=&ModuloInt(A=&ReturnParticleIDAttributes(), B=2))
     ))
   )
   &SkeletalMeshLocation(
    Bone Sampling Mode=::NewEnumerator4
    Inherit Velocity=true
    "Inherit Velocity (Scale)"=0.25
    MaskSpawnGroup=false
    Skeletal Mesh=$User.SKM_02
    VelocitySpeedLimit=false
   ) {module=SkeletalMeshLocation001}
   /Niagara/Modules/Spawn/Location/SphereLocation(MaskSpawnGroup=false, Module.Override Randomness=false, Sphere Radius=2, Surface Only=false)
   &AddVelocity(Velocity=&Multiply_VectorByFloat(
     Vector=$Output.SphereLocation.SphereVector
     Float=&RandomRangeFloat(Minimum=20, Maximum=30, Evaluation Type=::NewEnumerator1)
    )) {module=AddVelocity001}
   &InheritVelocity(
    Module.Limit Speed=false
    Velocity Limit=0
    Velocity Scale=&VectorFromFloat(Value=&RandomRangeFloat(Minimum=0.2, Maximum=0.4, Evaluation Type=::NewEnumerator1))
   )
  ]
  particleUpdateScript [
   &ParticleState()
   &ScaleSpriteSize(Scale Factor=&Vector2DFromFloat(Value=&FloatFromCurve(FloatCurve=DataInterface() {propertyValues {Curve {keys [
          <at 0 0 RCIM_Cubic>
          {arriveTangent=3.8197, interpMode=RCIM_Cubic, leaveTangent=3.8197, time=0.1034, value=1.1346}
          {arriveTangent=-0.4991, interpMode=RCIM_Cubic, leaveTangent=-0.4991, time=0.2059, value=0.7864}
          {arriveTangent=-0.9543, interpMode=RCIM_Cubic, leaveTangent=-0.9543, time=0.3019, value=1.0356}
          {arriveTangent=0.1149, interpMode=RCIM_Cubic, leaveTangent=0.1149, time=0.506, value=0.5}
          {arriveTangent=-0.8485, interpMode=RCIM_Cubic, leaveTangent=-0.8485, time=0.6432, value=1.0748}
          {arriveTangent=-3.012, interpMode=RCIM_Cubic, leaveTangent=-3.012, time=0.8021, value=0.2488}
          <at 1 0 RCIM_Cubic>
         ]}}})))
   &Drag(Drag=2, Use Linear Drag=true, Use Rotational Drag=false)
   &AccelerationForce(Acceleration=<0 0 -20>)
   &SolveForcesAndVelocity(Clamp Velocity=false, Limit Acceleration=false)
   &Color(Color=&MultiplyLinearColor(
     Linear Color=&LerpLinearColors(
      StartColor=$User.Color_01
      EndColor=$User.Color_02
      LerpFactor=&FloatFromCurve(FloatCurve=DataInterface() {propertyValues {Curve {keys [<at 0.2 0>,<at 0.8 1>]}}})
     )
     Scale Factor=&MakeVector4_FromLinearColor(LinearColor=&RandomRangeLinearColor(Minimum=<rgba 1 1 1 1>, Maximum=<rgba 0.65 0.65 1 1>))
    )) {module=Color002}
   &ScaleColor(
    Color Value To Scale=$Particles.Color
    Scale Alpha=&FloatFromCurve(FloatCurve=DataInterface() {propertyValues {Curve {keys [<at 0 0 RCIM_Cubic>,<at 0.1 1 RCIM_Cubic>,<at 1 0>]}}})
    Scale RGB=&VectorFromFloat(Value=25)
    ScaleA=true
    ScaleRGB=true
   ) {module=ScaleColor001}
  ]
  rendererClasses [&NiagaraSpriteRendererProperties]
  renderers [
   {rendererClass=&NiagaraSpriteRendererProperties}
   {rendererClass=&NiagaraSpriteRendererProperties, rendererIndex=1}
   {rendererClass=&NiagaraSpriteRendererProperties, rendererIndex=2}
   {rendererClass=&NiagaraSpriteRendererProperties, rendererIndex=3}
  ]
 }
}
summary {
 systemName=NS_ScholarEmote_Trail
 emitters [
  {emitterName=Sparkles, rendererClasses [&NiagaraSpriteRendererProperties]}
  {emitterName=RIbbon, rendererClasses [&NiagaraRibbonRendererProperties]}
  {emitterName=Halo, rendererClasses [&NiagaraSpriteRendererProperties]}
 ]
 userVariables [
  User.Color_01 {type=LinearColor, description="", defaultValue {struct=LinearColor, value=<rgba 0.6445 0.1981 0.6514 1>}}
  User.Color_02 {type=LinearColor, description="", defaultValue {struct=LinearColor, value=<rgba 0.227 0.0953 0.6514 1>}}
  User.RibbonIDOffset {type=NiagaraInt32}
  User.SkeletalMesh {
   type=NiagaraDataInterfaceSkeletalMesh
   defaultValue {
    struct=NiagaraExt_VariableValue_DataInterface
    value {dataInterface=&self:NiagaraDataInterfaceSkeletalMesh_0}
   }
  }
  User.SKM_02 {
   type=NiagaraDataInterfaceSkeletalMesh
   defaultValue {
    struct=NiagaraExt_VariableValue_DataInterface
    value {dataInterface=&self:NiagaraDataInterfaceSkeletalMesh_1}
   }
  }
  User.UseGreenMat {type=NiagaraFloat, defaultValue=0, description=""}
 ]
}
